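# Physics asset for the Sushi Chef backpack cosmetic, giving collision and joint limits to the pack and its lantern.
# A zero-length capsule, effectively a sphere, covers backpack_root, and a smaller capsule hangs below the dyn_lantern bone origin.
# One constraint joins the lantern to the root with narrow swing limits and a wider twist limit, restricting how far the lantern can dangle.
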PhysicsAsset /BRCosmetics/Accessories/FORT_Backpacks/M_MED_Sushi_Chef/Meshes/M_MED_Sushi_Chef_Backpack_Physics
bodies {
 backpack_root [{shapeType=Capsule, length=0, radius=16.3606, center=<-1.2356 -6.9773 -0.6654>, rotation=<rot 0 -90 0>}]
 dyn_lantern [{shapeType=Capsule, length=7.8546, radius=4.113, center=<0.1206 0.048 -7.2025>, rotation=<rot -1.9649 -90 -0.6174>}]
}
constraints [{bone1Name=dyn_lantern, bone2Name=backpack_root, swing1LimitDegrees=10, swing2LimitDegrees=10, twistLimitDegrees=45}]
